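# Niagara drift trail cosmetic for vehicles: stylized lightning at the wheels, with color sets for painted variants.
# Trail_Optimized draws two layered ribbons, stepping texture offsets over time and varying brightness and glow by color variant.
# ImpactSlash mesh slashes, persistent burst sprites and speed-driven, short-lived SubUV sprites follow skeletal bones or sockets, shifted by User.offset.
# Every emitter runs FN_DisableTireFX to mask effects per wheel.
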NiagaraSystem /VehicleCosmetics/DriftTrail/DriftTrail_StylizedLightning_/FX/NS_DriftTrail_StylizedLightning_MP
dependencies {
 usedDataInterfaces [
  &NiagaraDataInterfaceCurve
  &NiagaraDataInterfaceArrayFloat3
  &NiagaraDataInterfaceSpriteRendererInfo
  &NiagaraDataInterfaceSkeletalMesh
 ]
 usedDynamicInputs [&SelectVectorFromArray]
 usedModules [
  &self:SystemSpawnScript.SystemScriptSource.SystemScriptGraph.NiagaraNodeAssignment_0.SetVariables_#1
  &SystemState
  &FN_UnpackVehicleFlags
  /Game/Effects/Niagara/Modules/Vehicle/FN_UpdateVehicleAttribs
  /Game/Effects/Niagara/Modules/Vehicle/FN_VehicleFlags_Cheatcodes
 ]
}
emitters {
 ImpactSlash {
  emitterUpdateScript [
   &EmitterState()
   &SpawnRate(SpawnRate=50, Use Spawn Probability=false)
  ]
  particleSpawnScript [
   &VC_PaintedVariantsOverride(
    Black=<rgba 0.0052 0.0052 0.0052 1>
    BurntSienna=<rgba 0.0208 0.0064 0.0026 1>
    Cobalt=<rgba 0.0042 0.0092 0.2708 1>
    Crimson=<rgba 0.2604 0 0.0071 1>
    DefaultColor=<rgba 0.0625 0.2656 0.2526 1>
    Grey=<rgba 0.0729 0.0729 0.0729 1>
    Lime=<rgba 0.1191 0.5208 0 1>
    Orange=<rgba 0.2604 0.0908 0 1>
    Pink=<rgba 0.224 0.0268 0.1523 1>
    Purple=<rgba 0.0299 0 0.0729 1>
    Saffron=<rgba 0.3542 0.2715 0 1>
    SkyBlue=<rgba 0 0.2876 0.4323 1>
   )
   &InitializeParticle(
    Color=$Particles.Color
    Lifetime=&RandomRangeFloat(Minimum=0.1, Maximum=0.2)
    Mesh Scale Max=<2 2 5.2>
    Mesh Scale Min=<2 2.6 4.5>
    Mesh Scale Mode=::NewEnumerator2
    Sprite Rotation Mode=::NewEnumerator2
    UsePositionOffset=false
    Write Material Random=true
   )
   &PartitionParticles()
   &SkeletalMeshLocation(
    Bone / Socket Index=&Add_Integer(A=$Particles.PartitionParticles.Partition, B=4)
    Bone / Socket Sampling Mode=::NewEnumerator5
    Interpolate Mesh Sampling=false
    MaskSpawnGroup=false
    Mesh Sampling Type=::NewEnumerator6
    Sampled Position Offset=&AddVectorToPosition(Position=<-70 0 -20>, Vector=$User.offset)
    Skeletal Mesh=$System.SkeletalMesh
   )
   &InitialMeshOrientation(Use Rotation Vector=false)
  ]
  particleUpdateScript [
   &FN_DisableTireFX(
    Mask Index Attribute:NiagaraInt32=$Particles.PartitionParticles.Partition
    Mask Index FL=2
    Mask Index FR=3
    Mask Index RL=0
    Mask Index RR=1
   )
   &ParticleState()
   &SolveForcesAndVelocity(Clamp Velocity=false, Limit Acceleration=false)
   &SkeletalMeshLocation(
    Bone / Socket Index=&Add_Integer(A=$Particles.PartitionParticles.Partition, B=4)
    Bone / Socket Sampling Mode=::NewEnumerator5
    MaskSpawnGroup=false
    Mesh Sampling Type=::NewEnumerator6
    Sampled Position Offset=&AddVectorToPosition(Position=<-70 0 -20>, Vector=$User.offset)
    Skeletal Mesh=$System.SkeletalMesh
   ) {module=SkeletalMeshLocation001}
  ]
  rendererClasses [&NiagaraMeshRendererProperties, &NiagaraSpriteRendererProperties]
  renderers [
   {rendererClass=&NiagaraMeshRendererProperties, rendererIndex=0}
   {rendererClass=&NiagaraSpriteRendererProperties, rendererIndex=1}
  ]
 }
 SpritesV2 {
  emitterUpdateScript [
   &EmitterState(
    Inactive Response=::NewEnumerator1
    Life Cycle Mode=::NewEnumerator1
    Loop Behavior=::NewEnumerator1
    Loop Duration Mode=::NewEnumerator1
    UseLoopDelay=false
   )
   &SpawnRate(
    SpawnRate=/VehicleCosmetics/FX/Niagara/Modules/VC_Mask_Floats(
     FrontEnd Value:NiagaraFloat=200
     blsFrontEndPreview:NiagaraBool=false
     SpeedMin Value:NiagaraFloat=120
     SpeedMax Value:NiagaraFloat=180
     Speed Alpha:NiagaraFloat=$System.Speed_Alpha
    )
    Use Spawn Probability=false
   ) {enabled=false}
   &SpawnBurst_Instantaneous(Spawn Count=2, Spawn Group=1, Use Spawn Probability=false, UseLoopCountLimit=false) {module=SpawnBurst_Instantaneous001}
  ]
  particleSpawnScript [
   &VC_PaintedVariantsOverride(
    BurntSienna=<rgba 0.0642 0.0198 0.0079 1>
    Cobalt=<rgba 0.0237 0.0899 0.5052 1>
    DefaultColor=<rgba 0.0549 0 0.6667 1>
    Purple=<rgba 0 0 0 1>
    R1 Black:LinearColor=<rgba 0.0938 0.0938 0.0938 2.5>
    R1 BurntSienna:LinearColor=<rgba 0.855 0.2511 0.0891 1>
    R1 Cobalt:LinearColor=<rgba 0.4531 0.9611 1 1>
    R1 Crimson:LinearColor=<rgba 2.0312 0.127 0.1786 1>
    R1 DefaultColor=<rgba 1.011 0.6188 1.2 1>
    R1 ForestGreen:LinearColor=<rgba 0.3915 1.2 0.3875 1>
    R1 Grey:LinearColor=<rgba 3.5 3.5 3.5 1>
    R1 Lime:LinearColor=<rgba 3.2915 5 0.0029 1>
    R1 Orange:LinearColor=<rgba 3 2.5964 0 1>
    R1 Pink:LinearColor=<rgba 3 0.6562 2.1477 1>
    R1 Purple:LinearColor=<rgba 0.6315 0.0091 1.6 1>
    R1 Saffron:LinearColor=<rgba 1 0.8627 0.4115 1>
    R1 SkyBlue:LinearColor=<rgba 0.487 0.8949 1.1 1>
    R1 TitaniumWhite:LinearColor=<rgba 2.5 2.5 2.5 1>
    R2 BurntSienna:LinearColor=<rgba 0.0642 0.0198 0.0079 1>
    R2 Cobalt:LinearColor=<rgba 0 0.0185 1 1>
    R2 DefaultColor:LinearColor=<rgba 0.0549 0 0.6667 1>
    R2 Purple:LinearColor=<rgba 0.2789 0 0.6795 0.85>
    R2 Saffron:LinearColor=<rgba 0.8854 0.7326 0.2306 1>
    R2 SkyBlue:LinearColor=<rgba 0 0.2946 0.4427 1>
    R2 TitaniumWhite:LinearColor=<rgba 0.4948 0.4948 0.4948 1>
    TitaniumWhite=<rgba 0.8177 0.8177 0.8177 1>
    Use R1 Painted Overrides=true
    Use R2 Painted Overrides=true
   )
   /SetVariables(Particles.NewTagVis:NiagaraInt32=&SpawnGroup()) {isSetParameters=true, module=SetVariables_#2}
   &InitializeParticle(
    Color=$Particles.Color
    Sprite Rotation Mode=::NewEnumerator2
    Sprite Size Mode=::NewEnumerator3
    Uniform Sprite Size=100
    UsePositionOffset=false
    Write Material Random=true
   )
   &PartitionParticles()
   &SkeletalMeshLocation(
    Bone / Socket Index=&Add_Integer(A=$Particles.PartitionParticles.Partition, B=4)
    Bone / Socket Sampling Mode=::NewEnumerator5
    Interpolate Mesh Sampling=false
    MaskSpawnGroup=false
    Mesh Sampling Type=::NewEnumerator6
    Sampled Position Offset=&AddVectorToPosition(Position=<-70 0 -20>, Vector=$User.offset)
    Skeletal Mesh=$System.SkeletalMesh
   )
   &AddVelocity(Velocity=&VC_Mask_Vectors(
     FrontEnd Value:Vector3f=<-1500 0 0>
     blsFrontEndPreview:NiagaraBool=false
     SpeedMin Value:Vector3f=<-800 0 0>
     SpeedMax Value:Vector3f=<-1 0 0>
     Speed Alpha:NiagaraFloat=$System.Speed_Alpha
    )) {enabled=false}
  ]
  particleUpdateScript [
   &FN_DisableTireFX(
    Mask Index Attribute:NiagaraInt32=$Particles.PartitionParticles.Partition
    Mask Index FL=2
    Mask Index FR=3
    Mask Index RL=0
    Mask Index RR=1
   )
   &ParticleState(Kill Particles When Lifetime Has Elapsed=false)
   &SkeletalMeshLocation(
    Bone / Socket Index=&Add_Integer(A=$Particles.PartitionParticles.Partition, B=4)
    Bone / Socket Sampling Mode=::NewEnumerator5
    MaskSpawnGroup=false
    Mesh Sampling Type=::NewEnumerator6
    Sampled Position Offset=&AddVectorToPosition(Position=<-70 0 -20>, Vector=$User.offset)
    Skeletal Mesh=$System.SkeletalMesh
   ) {module=SkeletalMeshLocation001}
   &SolveForcesAndVelocity(Clamp Velocity=false, Limit Acceleration=false)
  ]
  rendererClasses [&NiagaraSpriteRendererProperties]
  renderers [
   {rendererClass=&NiagaraSpriteRendererProperties}
   {rendererClass=&NiagaraSpriteRendererProperties, rendererIndex=1}
   {rendererClass=&NiagaraSpriteRendererProperties, rendererIndex=2}
   {rendererClass=&NiagaraSpriteRendererProperties, rendererIndex=3}
   {rendererClass=&NiagaraSpriteRendererProperties, rendererIndex=4}
  ]
 }
 SpritesV2001 {
  emitterUpdateScript [
   &EmitterState(Inactive Response=::NewEnumerator1, Life Cycle Mode=::NewEnumerator1, UseLoopDelay=false)
   &SpawnRate(
    SpawnRate=&VC_Mask_Floats(
     FrontEnd Value:NiagaraFloat=200
     blsFrontEndPreview:NiagaraBool=false
     SpeedMin Value:NiagaraFloat=120
     SpeedMax Value:NiagaraFloat=180
     Speed Alpha:NiagaraFloat=$System.Speed_Alpha
    )
    Use Spawn Probability=false
   )
  ]
  particleSpawnScript [
   &VC_PaintedVariantsOverride(
    Black=<rgba 0.0191 0.0191 0.0191 3>
    BurntSienna=<rgba 0.0625 0.0193 0.0077 3>
    Cobalt=<rgba 0 0.0185 1 3>
    Crimson=<rgba 0.4635 0 0.0126 3>
    DefaultColor=<rgba 0.3802 0.9868 1 3>
    ForestGreen=<rgba 0 0.5 0.0253 3>
    Grey=<rgba 0.2031 0.2031 0.2031 3>
    Lime=<rgba 0.2943 0.5885 0 3>
    Orange=<rgba 0.6724 0.2346 0 3>
    Pink=<rgba 0.5052 0.0605 0.3435 3>
    Purple=<rgba 0.0962 0 0.2344 3>
    Saffron=<rgba 0.9 0.69 0 3>
    SkyBlue=<rgba 0 0.357 0.5365 3>
    TitaniumWhite=<rgba 0.75 0.75 0.75 3>
   )
   /SetVariables(Particles.NewTagVis:NiagaraInt32=&SpawnGroup()) {isSetParameters=true, module=SetVariables_#2}
   &InitializeParticle(
    Color=$Particles.Color
    Lifetime=&RandomRangeFloat(Minimum=0.011, Maximum=0.05)
    Sprite Rotation Mode=::NewEnumerator2
    Sprite Size Mode=::NewEnumerator3
    Uniform Sprite Size=23
    UsePositionOffset=false
    Write Material Random=true
   )
   &PartitionParticles()
   &SkeletalMeshLocation(
    Bone / Socket Index=&Add_Integer(A=$Particles.PartitionParticles.Partition, B=4)
    Bone / Socket Sampling Mode=::NewEnumerator5
    Interpolate Mesh Sampling=false
    MaskSpawnGroup=false
    Mesh Sampling Type=::NewEnumerator6
    Sampled Position Offset=&AddVectorToPosition(Position=<-70 0 -20>, Vector=$User.offset)
    Skeletal Mesh=$System.SkeletalMesh
   )
   &AddVelocity(Velocity=&VC_Mask_Vectors(
     FrontEnd Value:Vector3f=<-1500 0 0>
     blsFrontEndPreview:NiagaraBool=false
     SpeedMin Value:Vector3f=<-800 0 0>
     SpeedMax Value:Vector3f=<-1 0 0>
     Speed Alpha:NiagaraFloat=$System.Speed_Alpha
    )) {enabled=false}
   &ShapeLocation(
    MaskSpawnGroup=false
    Override Seed=false
    Sphere Radius=/VehicleCosmetics/FX/Niagara/Modules/MergeEmitters/VC_FloatBySpawnGroup(
     SpawnGroup or ExecIndex:NiagaraInt32=0
     CustomInt:NiagaraInt32=0
     MaximumSpawnGroups:NiagaraInt32=2
     Float SG 0:NiagaraFloat=20
     Float SG 1:NiagaraFloat=0
     Float SG 2:NiagaraFloat=0
     Float SG 3:NiagaraFloat=0
     Float SG 4:NiagaraFloat=0
     Float SG 5:NiagaraFloat=0
    )
    UseOwnerScale=true
   )
  ]
  particleUpdateScript [
   &FN_DisableTireFX(
    Mask Index Attribute:NiagaraInt32=$Particles.PartitionParticles.Partition
    Mask Index FL=2
    Mask Index FR=3
    Mask Index RL=0
    Mask Index RR=1
   )
   &ParticleState()
   &SkeletalMeshLocation(
    Bone / Socket Index=&Add_Integer(A=$Particles.PartitionParticles.Partition, B=4)
    Bone / Socket Sampling Mode=::NewEnumerator5
    MaskSpawnGroup=false
    Mesh Sampling Type=::NewEnumerator6
    Sampled Position Offset=&AddVectorToPosition(Position=<-70 0 -20>, Vector=$User.offset)
    Skeletal Mesh=$System.SkeletalMesh
   ) {module=SkeletalMeshLocation001}
   &ShapeLocation(
    MaskSpawnGroup=false
    Override Seed=false
    Shape Origin=$StackContext.Position
    Sphere Radius=&VC_FloatBySpawnGroup(
     SpawnGroup or ExecIndex:NiagaraInt32=0
     CustomInt:NiagaraInt32=0
     MaximumSpawnGroups:NiagaraInt32=2
     Float SG 0:NiagaraFloat=20
     Float SG 1:NiagaraFloat=0
     Float SG 2:NiagaraFloat=0
     Float SG 3:NiagaraFloat=0
     Float SG 4:NiagaraFloat=0
     Float SG 5:NiagaraFloat=0
    )
    UseOwnerScale=true
   ) {module=ShapeLocation001}
   &SolveForcesAndVelocity(Clamp Velocity=false, Limit Acceleration=false)
   &SubUVAnimation(
    Output Alternate Renderer Binding=true
    "Sprite Renderer "=DataInterface() {propertyValues {SpriteRenderer=&self:SpritesV2001_0.NiagaraSpriteRendererProperties_5}}
    UseEndFrame=false
    UseStartFrame=false
   )
  ]
  rendererClasses [&NiagaraSpriteRendererProperties]
  renderers [{rendererClass=&NiagaraSpriteRendererProperties}]
 }
 Trail_Optimized {
  emitterSpawnScript [/SetVariables(
    Emitter.Brightness:NiagaraFloat=&SelectFloatFromArray(
     Direct Array Index=$System.ColorVariantID
     Float Selection Array=DataInterface() {propertyValues {FloatData [0.8,1,1,1.75,1.45,1.95,1,1,1,1,1.02,1,0.99,1,3]}}
    )
    Emitter.ContrastText1:NiagaraFloat=&SelectFloatFromArray(
     Direct Array Index=$System.ColorVariantID
     Float Selection Array=DataInterface() {propertyValues {FloatData [0.5,1,1,1,1,1,1,1,1,1,1,1,1,1,1]}}
    )
    Emitter.Glow:NiagaraFloat=&SelectFloatFromArray(
     Direct Array Index=$System.ColorVariantID
     Float Selection Array=DataInterface() {propertyValues {FloatData [0.75,2,0.5,0.5,0.3,0.5,0.5,0.5,0.5,0.45,0.5,0.6,0.5,0.5,0.5]}}
    )
   ) {isSetParameters=true}]
  emitterUpdateScript [
   &EmitterState()
   &VC_RateQualityHandler(Epic Spawn Rate=90, High Spawn Rate=60, Low Spawn Rate=30, Medium Spawn Rate=45, Use Custom Rate?=true)
   /SetVariables(
    Emitter.TexOffset2=HlslExpression() {hlslExpression="floor(fmod(Engine.Time * 10 , 8)) * .125"}
    Emitter.TexOffset3=HlslExpression() {hlslExpression="floor(fmod(Engine.Time * 15 , 8)) * .125"}
   ) {isSetParameters=true}
   &VC_SpawnRibbon_v1(Number of Ribbons=2, Spawn Time=-1, SpawnRate=100)
  ]
  particleSpawnScript [
   &VC_PaintedVariantsOverride(
    Black=<rgba 0.1406 0.1406 0.1406 0.5>
    BurntSienna=<rgba 0.3385 0.2051 0.1693 0.2>
    Cobalt=<rgba 0.0581 0.0669 0.5312 0.5>
    Crimson=<rgba 0.4167 0.0629 0.0725 0.3>
    DefaultColor=<rgba 0.127 0.3117 0.5417 1>
    ForestGreen=<rgba 0.1823 0.625 0.2047 0.15>
    Grey=<rgba 0.3073 0.3073 0.3073 0.15>
    Lime=<rgba 0.3317 0.5156 0.1477 0.25>
    Orange=<rgba 0.4792 0.2175 0.0774 0.35>
    Pink=<rgba 0.4635 0.128 0.3415 0.2>
    Purple=<rgba 0.2019 0.1177 0.3229 0.25>
    R1 Black:LinearColor=<rgba 0.2448 0.2448 0.2448 2>
    R1 BurntSienna:LinearColor=<rgba 0.4323 0.2583 0.2116 1.5>
    R1 Cobalt:LinearColor=<rgba 0.1594 0.1897 1.8 2>
    R1 Crimson:LinearColor=<rgba 0.9062 0.1841 0.2037 2>
    R1 DefaultColor=<rgba 0.5201 0.3359 1.5 2>
    R1 ForestGreen:LinearColor=<rgba 0.1215 0.6667 0.1491 2>
    R1 Grey:LinearColor=<rgba 0.6615 0.6615 0.6615 2>
    R1 Lime:LinearColor=<rgba 0.4482 0.7083 0.1882 2>
    R1 Orange:LinearColor=<rgba 1.5 0.7572 0.3594 1>
    R1 Pink:LinearColor=<rgba 1.5 0.4141 1.1051 1>
    R1 Purple:LinearColor=<rgba 0.6819 0.2865 1.25 1>
    R1 Saffron:LinearColor=<rgba 1.5 1.2138 0.2734 1>
    R1 SkyBlue:LinearColor=<rgba 0.3516 1.1157 1.5 1>
    R1 TitaniumWhite:LinearColor=<rgba 1.5 1.5 1.5 1>
    R2 Black:LinearColor=<rgba 0.2396 0.2396 0.2396 1.8>
    R2 BurntSienna:LinearColor=<rgba 0.355 0.1769 0.0573 1.35>
    R2 Cobalt:LinearColor=<rgba 0 0.1782 1.6226 0.6>
    R2 Crimson:LinearColor=<rgba 1 0.0417 0.0676 2>
    R2 DefaultColor:LinearColor=<rgba 1.3148 1.3843 3.5 0.5826>
    R2 ForestGreen:LinearColor=<rgba 0.1524 1.6 0.125 1>
    R2 Grey:LinearColor=<rgba 0.75 0.75 0.75 1>
    R2 Lime:LinearColor=<rgba 1.1126 1.7057 0 1.5>
    R2 Orange:LinearColor=<rgba 1.6 0.9048 0.0417 1.8>
    R2 Pink:LinearColor=<rgba 1.4804 0.225 1.5 1.5>
    R2 Purple:LinearColor=<rgba 0.4659 0.0095 0.75 1.5>
    R2 Saffron:LinearColor=<rgba 1.7188 1.5043 0.0269 1.5>
    R2 SkyBlue:LinearColor=<rgba 0.0792 1.2907 1.9 1>
    R2 TitaniumWhite:LinearColor=<rgba 1.5 1.5 1.5 0.5>
    R3 Black:LinearColor=<rgba 0.2396 0.2396 0.2396 1.8>
    R3 BurntSienna:LinearColor=<rgba 0.355 0.1769 0.0573 1.35>
    R3 Cobalt:LinearColor=<rgba 0 0.1782 1.6226 0.6>
    R3 Crimson:LinearColor=<rgba 1 0.0417 0.0676 2>
    R3 DefaultColor:LinearColor=<rgba 0.6517 1.6303 1.8958 1.8>
    R3 ForestGreen:LinearColor=<rgba 0.1524 1.6 0.125 1>
    R3 Grey:LinearColor=<rgba 0.75 0.75 0.75 1>
    R3 Lime:LinearColor=<rgba 1.1126 1.7057 0 1.5>
    R3 Orange:LinearColor=<rgba 1.6 0.9048 0.0417 1.8>
    R3 Pink:LinearColor=<rgba 1.4804 0.225 1.5 1.5>
    R3 Purple:LinearColor=<rgba 0.4659 0.0095 0.75 1.5>
    R3 Saffron:LinearColor=<rgba 1.7188 1.5043 0.0269 1.5>
    R3 SkyBlue:LinearColor=<rgba 0.0792 1.2907 1.9 1>
    R3 TitaniumWhite:LinearColor=<rgba 1.5 1.5 1.5 0.5>
    R4 Black:LinearColor=<rgba 0.4948 0.4948 0.4948 0.2>
    R4 BurntSienna:LinearColor=<rgba 0.349 0.1847 0.0891 0.4226>
    R4 Cobalt:LinearColor=<rgba 0 0.2128 1.9375 0.1>
    R4 Crimson:LinearColor=<rgba 1.2 0.3188 0.3426 0.2>
    R4 DefaultColor:LinearColor=<rgba 0.5 0.5 0.5 0.0609>
    R4 ForestGreen:LinearColor=<rgba 0.2624 1.15 0.2456 0.5>
    R4 Grey:LinearColor=<rgba 0.8021 0.8021 0.8021 0.5>
    R4 Lime:LinearColor=<rgba 1.1126 1.7057 0 0.2>
    R4 Orange:LinearColor=<rgba 1.2 0.7803 0.375 0.5>
    R4 Pink:LinearColor=<rgba 1.2 0.5 0.9455 0.5>
    R4 Purple:LinearColor=<rgba 0.4542 0 1.1068 0.3>
    R4 Saffron:LinearColor=<rgba 1.8 1.6277 0.4406 0.3217>
    R4 SkyBlue:LinearColor=<rgba 0 1.2996 1.9531 0.25>
    R4 TitaniumWhite:LinearColor=<rgba 0.75 0.75 0.75 0.3>
    Saffron=<rgba 0.9 0.7283 0.1641 0.25>
    SkyBlue=<rgba 0.1383 0.282 0.3542 0.35>
    TitaniumWhite=<rgba 0.4167 0.4167 0.4167 0.4>
    Use R1 Painted Overrides=true
    Use R2 Painted Overrides=true
    Use R3 Painted Overrides=true
    Use R4 Painted Overrides=true
   )
   &VC_InitializeRibbon_v1(
    Color=$Particles.Color
    Custom Offset=true
    Enable Skeletal Position=true
    Lifetime=0.15
    Local Vector Offset=&AddVectorToPosition(Position=<-60 0 -15>, Vector=$User.offset)
    Offset Amount=1
    Ribbon Width=40
   )
   &AddVelocity(Velocity=<-2500 0 0>, Velocity Speed Scale=2) {module=AddVelocity001}
  ]
  particleUpdateScript [
   &FN_DisableTireFX(Mask Index Attribute:NiagaraInt32=$Particles.SpawnGroup, Mask Index FL=2, Mask Index FR=3, Mask Index RL=0, Mask Index RR=1)
   &InheritSourceMovement(
    Applied Position Scale=&Multiply_VectorByFloat(
     Vector=<1 1 1>
     Float=&FloatFromCurve(FloatCurve=DataInterface() {propertyValues {Curve {keys [{interpMode=RCIM_Cubic, tangentMode=RCTM_User, time=0.3, value=1}, {interpMode=RCIM_Cubic, tangentMode=RCTM_User, time=0.7, value=0.5}]}}})
    )
    ApplyForce=false
    ApplyPosition=true
    ApplyPositionFalloff=false
   )
   &ParticleState()
   &SolveForcesAndVelocity(Clamp Velocity=false, Limit Acceleration=false)
   &ScaleRibbonWidth()
   &ScaleColor(
    Scale RGB=&MakeCustomVectorFromBool(Bool=$User.bIsFrontendPreview, True Vector=<5 5 5>, False Vector=<5 5 5>)
    ScaleA=true
    ScaleRGB=true
   )
   &VC_UpdateRibbon_v1(Enable Skeletal Position=true)
  ]
  rendererClasses [&NiagaraRibbonRendererProperties]
  renderers [
   {rendererClass=&NiagaraRibbonRendererProperties}
   {rendererClass=&NiagaraRibbonRendererProperties, rendererIndex=1}
   {rendererClass=&NiagaraRibbonRendererProperties, rendererIndex=2}
   {rendererClass=&NiagaraRibbonRendererProperties, rendererIndex=3}
   {rendererClass=&NiagaraRibbonRendererProperties, rendererIndex=4}
  ]
 }
}
summary {
 systemName=NS_DriftTrail_StylizedLightning_MP
 emitters [
  {emitterName=Trail_Optimized, rendererClasses [&NiagaraRibbonRendererProperties]}
  {emitterName=SpritesV2, rendererClasses [&NiagaraSpriteRendererProperties]}
  {emitterName=ImpactSlash, rendererClasses [&NiagaraMeshRendererProperties, &NiagaraSpriteRendererProperties]}
  {emitterName=SpritesV2001, rendererClasses [&NiagaraSpriteRendererProperties]}
 ]
 userVariables [
  User.bIsFrontendPreview {type=NiagaraBool}
  User.bIsPainted {type=NiagaraBool}
  User.ColorVariantID {type=NiagaraInt32}
  User.offset {type=Vector3f, description="offset from wheels(set through cosmetic template)", defaultValue {struct=Vector3f, value=<0 0 0>}}
  User.PaintedColor {type=LinearColor, description="", defaultValue {struct=LinearColor, value=<rgba 1 1 1 1>}}
  User.PaintedScalar {type=NiagaraFloat, defaultValue=1, description=""}
  User.VehicleFlags {type=NiagaraInt32, defaultValue=806289408}
  User.Wheel_Radius_Array {
   type=NiagaraDataInterfaceArrayFloat
   defaultValue {
    struct=NiagaraExt_VariableValue_DataInterface
    value {
     dataInterface=&self:NiagaraDataInterfaceArrayFloat_0
     dataInterfaceClass=&NiagaraDataInterfaceArrayFloat
    }
   }
  }
  User.Wheel_Width_Array {
   type=NiagaraDataInterfaceArrayFloat
   defaultValue {
    struct=NiagaraExt_VariableValue_DataInterface
    value {
     dataInterface=&self:NiagaraDataInterfaceArrayFloat_1
     dataInterfaceClass=&NiagaraDataInterfaceArrayFloat
    }
   }
  }
  User.WheelSize {type=NiagaraFloat, defaultValue=0.1, description=""}
 ]
}
