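# Unlit translucent glow material for Niagara sprites in the stylized lightning drift trail vehicle cosmetic.
# Particle color passes through the shared MF_FX_EmissiveAdjustments function to form the glow color.
# A circular gradient mask, multiplied by particle alpha and a depth fade function, shapes a round opacity that softens where sprites meet geometry.
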Material /VehicleCosmetics/DriftTrail/DriftTrail_StylizedLightning_/Materials/Glow_MAT
graph {
 properties {bUsedWithNiagaraSprites=true, blendMode=BLEND_Translucent, pixelDepthOffsetMode=PDOM_Legacy, shadingModel=MSM_Unlit, usedShadingModels=Unlit}
 nodes [
  n1 = ParticleColor()
  MaterialFunctionCall(Emissive=@n1.RGBA) {materialFunction=&MF_FX_EmissiveAdjustments}
  Multiply(
   A=Multiply(
    A=TextureSample() {samplerType=SAMPLERTYPE_LinearColor, texture=/DealershipSharedLibrary/FX/Textures/Gradient_Masks/T_VCVFX_LinearCircleGradient_A}.R
    B=@n1.A
   )
   B=MaterialFunctionCall(Opacity In=Constant() {r=1}) {materialFunction=&MF_VCVFX_DepthFade}.Result
  )
 ]
}
